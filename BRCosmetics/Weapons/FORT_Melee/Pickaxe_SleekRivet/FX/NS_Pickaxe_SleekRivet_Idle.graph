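# Niagara idle effect for the SleekRivet pickaxe that attaches warm gold sprite glows to the weapon's skeletal mesh.
# EdgeGlow bursts 32 sprites and traces a curved line, with curves over normalized execution index driving X and Z offsets.
# NormalGlows bursts eight sprites, using partition index to pick color, size and hand-placed offsets from arrays.
# Both emitters re-sample a bone or socket position every update and keep particles past their lifetime.
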NiagaraSystem /BRCosmetics/Weapons/FORT_Melee/Pickaxe_SleekRivet/FX/NS_Pickaxe_SleekRivet_Idle
dependencies {
 usedDataInterfaces [
  &NiagaraDataInterfaceCurve
  &NiagaraDataInterfaceArrayColor
  &NiagaraDataInterfaceArrayFloat
  &NiagaraDataInterfaceArrayFloat3
  &NiagaraDataInterfaceSkeletalMesh
 ]
 usedDynamicInputs []
 usedModules [
  &self:SystemSpawnScript.SystemScriptSource.SystemScriptGraph.NiagaraNodeAssignment_0.SetVariables_#1
  &SystemState
 ]
}
emitters {
 EdgeGlow {
  emitterUpdateScript [
   &EmitterState(Life Cycle Mode=::NewEnumerator1, Loop Behavior=::NewEnumerator1, Loop Duration Mode=::NewEnumerator1, UseLoopDelay=false)
   &SpawnBurst_Instantaneous(Spawn Count=32, Use Spawn Probability=false, UseLoopCountLimit=false)
  ]
  particleSpawnScript [&InitializeParticle(Color=<rgba 1 0.6897 0.097 0.1>, Lifetime=5, Sprite Size Mode=::NewEnumerator3, UsePositionOffset=false, Write Material Random=true)]
  particleUpdateScript [
   &SkeletalMeshLocation(
    Bone / Socket Index=2
    MaskSpawnGroup=false
    Mesh Sampling Type=::NewEnumerator5
    Sampled Position Offset=&MakeVector(
     X=&FloatFromCurve(
      FloatCurve=DataInterface() {propertyValues {Curve {keys [
          {arriveTangent=1.3829, time=0.0204, value=0.5673}
          <at 0.2264 0.95>
          {arriveTangent=-0.3571, leaveTangent=-0.3571, time=0.7, value=0.95}
          {arriveTangent=-0.4534, time=1, value=0.7134}
         ]}}}
      CurveIndex=&ReturnNormalizedExecIndex()
      Scale Curve=27.9157
     )
     Y=0
     Z=&FloatFromCurve(
      FloatCurve=DataInterface() {propertyValues {Curve {keys [<at 0 -0.3796>,<at 0.2041 -0.45>,<at 0.3 -0.35>,<at 0.7 0.375>,<at 1 0.7366>]}}}
      CurveIndex=&ReturnNormalizedExecIndex()
      Scale Curve=50.2093
     )
    )
    Skeletal Mesh=$System.Skeletal Mesh
    Socket Sampling Mode=::NewEnumerator5
   ) {module=SkeletalMeshLocation001}
   &ParticleState(Kill Particles When Lifetime Has Elapsed=false, Let Infinitely Lived Particles Die When Emitter Deactivates=true)
  ]
  rendererClasses [&NiagaraSpriteRendererProperties]
  renderers [{rendererClass=&NiagaraSpriteRendererProperties}]
 }
 NormalGlows {
  emitterUpdateScript [
   &EmitterState(Life Cycle Mode=::NewEnumerator1, Loop Behavior=::NewEnumerator1, Loop Duration Mode=::NewEnumerator1, UseLoopDelay=false)
   &SpawnBurst_Instantaneous(Spawn Count=8, Use Spawn Probability=false, UseLoopCountLimit=false)
  ]
  particleSpawnScript [
   &PartitionParticles(Number Of Partitions=8)
   &InitializeParticle(
    Color=&SelectLinearColorFromArray(
     Array Sampling Mode=::NewEnumerator1
     Direct Array Index=$Particles.PartitionParticles.Partition
     Color Selection Array=DataInterface() {propertyValues {ColorData [
        <rgba 1 0.6897 0.097 0.33>
        <rgba 1 0.6897 0.097 0.33>
        <rgba 1 0.6897 0.097 0.33>
        <rgba 1 0.6897 0.097 0.33>
        <rgba 1 0.6897 0.097 1>
        <rgba 1 0.6897 0.097 1>
        <rgba 2 1.3794 0.1941 1>
        <rgba 2 1.3794 0.1941 1>
       ]}}
     Random Seed=?
    )
    Lifetime=5
    Sprite Size Mode=::NewEnumerator3
    Uniform Sprite Size=&SelectFloatFromArray(
     Direct Array Index=$Particles.PartitionParticles.Partition
     Float Selection Array=DataInterface() {propertyValues {FloatData [7,7,6,6,8,8,14,14]}}
    )
    UsePositionOffset=false
    Write Material Random=true
   )
   &CameraOffset(Camera Offset Amount=2)
  ]
  particleUpdateScript [
   &SkeletalMeshLocation(
    Bone / Socket Index=2
    MaskSpawnGroup=false
    Mesh Sampling Type=::NewEnumerator5
    Sampled Position Offset=&SelectVectorFromArray(
     Direct Array Index=$Particles.PartitionParticles.Partition
     Vector Selection Array=DataInterface() {propertyValues {FloatData [<-19 4.25 -3.5>, <-19 -4.25 -3.5>, <-13 3 -12.5>, <-13 -3 -12.5>, <-9.5 6 8.8685>, <-9.5 -6 8.8685>, <-1.4111 0 -30>, <-1.5 0 -23.1406>]}}
    )
    Skeletal Mesh=$System.Skeletal Mesh
    Socket Sampling Mode=::NewEnumerator5
   ) {module=SkeletalMeshLocation001}
   &ParticleState(Kill Particles When Lifetime Has Elapsed=false, Let Infinitely Lived Particles Die When Emitter Deactivates=true)
  ]
  rendererClasses [&NiagaraSpriteRendererProperties]
  renderers [{rendererClass=&NiagaraSpriteRendererProperties}]
 }
}
summary {
 systemName=NS_Pickaxe_SleekRivet_Idle
 emitters [
  {emitterName=EdgeGlow, rendererClasses [&NiagaraSpriteRendererProperties]}
  {emitterName=NormalGlows, rendererClasses [&NiagaraSpriteRendererProperties]}
 ]
 userVariables []
}
