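# Niagara weapon trail effect for the SleekRivet pickaxe, built as a single BladeTrailDark ribbon emitter.
# FN anim-trail modules spawn short-lived particles and position them from two bone or socket indices on a skeletal mesh.
# A warm gold, time-of-day color dims over life, while a separate DarktrailColor parameter holds black with fading alpha.
# Three ribbon renderers draw the emitter's particles.
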NiagaraSystem /BRCosmetics/Weapons/FORT_Melee/Pickaxe_SleekRivet/FX/NS_Pickaxe_SleekRivet_Trails
dependencies {
 usedDataInterfaces [&NiagaraDataInterfaceColorCurve, &NiagaraDataInterfaceCurve, &NiagaraDataInterfaceSkeletalMesh]
 usedDynamicInputs []
 usedModules [
  &self:SystemSpawnScript.SystemScriptSource.SystemScriptGraph.NiagaraNodeAssignment_0.SetVariables_#1
  &SystemState
 ]
}
emitters {BladeTrailDark {
  emitterUpdateScript [
   &EmitterState()
   &FN_SpawnAnimTrail(Spawn Interval=0)
  ]
  particleSpawnScript [
   &FN_InitializeAnimTrailBoneIndecies(First Bone or Socket Index=1, Second Bone or Socket Index=0)
   &FN_InitializeAnimTrail(
    Color=&FNTimeOfDayColor(
     Night Color=<rgba 1 0.7943 0.1771 0.66>
     Morning Color=<rgba 1 0.7943 0.1771 1>
     Day Color=<rgba 1 0.7943 0.1771 1>
     Evening Color=<rgba 1 0.7943 0.1771 0.8>
    )
    Lifetime=0.13
    SkeletalMesh=$System.Skeletal Mesh
   )
  ]
  particleUpdateScript [
   &FN_UpdateAnimTrail(Locked Particle RibbonWidth=1, SkeletalMesh=$User.SkeletalMesh) {module=FN_UpdateAnimTrail001}
   &ScaleColor(
    Scale Alpha=&FloatFromCurve(
     FloatCurve=DataInterface() {propertyValues {Curve {keys [{interpMode=RCIM_Cubic, tangentMode=RCTM_User, time=0, value=1}, {interpMode=RCIM_Cubic, tangentMode=RCTM_User, time=1, value=0}]}}}
     Scale Curve=0.2
    )
    Scale RGB=&MakeVectorFromLinearColorRGB(Color=&ColorFromCurve(ColorCurve=DataInterface() {propertyValues {
        AlphaCurve {keys []}
        BlueCurve {keys [{arriveTangent=-1.3253, interpMode=RCIM_Cubic, time=0.077, value=1}, <at 0.9809 0.5>]}
        GreenCurve {keys [{arriveTangent=-1.3253, interpMode=RCIM_Cubic, time=0.077, value=1}, <at 0.9809 0.5>]}
        RedCurve {keys [{arriveTangent=-1.3253, interpMode=RCIM_Cubic, time=0.077, value=1}, <at 0.9809 0.5>]}
       }}))
    ScaleA=true
    ScaleRGB=true
   )
   /SetVariables(Particles.DarktrailColor:LinearColor=&ColorFromCurve(ColorCurve=DataInterface() {propertyValues {
       AlphaCurve {keys [<at 0 1 RCIM_Cubic>,<at 0.7125 0 RCIM_Cubic>]}
       BlueCurve {keys [<at 0 0 RCIM_Cubic>,<at 1 0 RCIM_Cubic>]}
       GreenCurve {keys [<at 0 0 RCIM_Cubic>,<at 1 0 RCIM_Cubic>]}
       RedCurve {keys [<at 0 0 RCIM_Cubic>,<at 1 0 RCIM_Cubic>]}
      }})) {isSetParameters=true}
  ]
  rendererClasses [&NiagaraRibbonRendererProperties]
  renderers [
   {rendererClass=&NiagaraRibbonRendererProperties}
   {rendererClass=&NiagaraRibbonRendererProperties, rendererIndex=1}
   {rendererClass=&NiagaraRibbonRendererProperties, rendererIndex=2}
  ]
 }}
summary {
 systemName=NS_Pickaxe_SleekRivet_Trails
 emitters [{emitterName=BladeTrailDark, rendererClasses [&NiagaraRibbonRendererProperties]}]
 userVariables [
  User.RibbonIDOffset {type=NiagaraInt32}
  User.Variant {type=NiagaraFloat, defaultValue=0, description=""}
 ]
}
