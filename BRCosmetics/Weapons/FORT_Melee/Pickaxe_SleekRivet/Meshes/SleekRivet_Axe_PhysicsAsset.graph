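# Physics asset for the SleekRivet two-handed melee pickaxe skeletal mesh, a cosmetic weapon.
# A single capsule body on the Melee_TwoHanded_Handle bone forms the whole collision shape, offset along and aligned to the tool.
# It has no constraints, so the pickaxe simulates and collides as one rigid piece.
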PhysicsAsset /BRCosmetics/Weapons/FORT_Melee/Pickaxe_SleekRivet/Meshes/SleekRivet_Axe_PhysicsAsset
bodies {Melee_TwoHanded_Handle [{shapeType=Capsule, length=80.7535, radius=26.0445, center=<0.7025 1.9068 38.3421>, rotation=<rot -0.6169 -90 0.0004>}]}
constraints []
